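; Blueprint for the Asteria cherry blossom tree, variant B with a base, used as environment foliage.
; Its construction script runs the parent's script, then forces the Base component to unit world scale and zero world rotation.
; This keeps the base unscaled and unrotated in world space, whatever the actor's own scale and rotation.
; BeginPlay, overlap and Tick are disabled placeholders, so the graph adds no runtime logic.
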(event EventBeginPlay
  (|CallParentFunction) ; This node is disabled and will not be called.\nDrag off pins to build functionality.
)

(event Collision|EventActorBeginOverlap
  (params (OtherActor :type Actor%20Object%20Reference))
  (|CallParentFunction OtherActor) ; This node is disabled and will not be called.\nDrag off pins to build functionality.
)

(event EventTick
  (params (DeltaSeconds :type Float%20%28single-precision%29))
  (|CallParentFunction DeltaSeconds) ; This node is disabled and will not be called.\nDrag off pins to build functionality.
)

(fn ConstructionScript
  (bind _base (Variables|Default|GetBase))
  (|CallParentFunction)
  (Transformation|SetWorldScale3D _base "1.000000,1.000000,1.000000")
  (Transformation|SetWorldRotation _base "0, 0, 0" false false))
